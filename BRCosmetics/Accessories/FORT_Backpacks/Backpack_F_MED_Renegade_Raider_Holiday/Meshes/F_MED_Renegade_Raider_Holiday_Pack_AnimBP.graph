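# Animation Blueprint for the Renegade Raider Holiday backpack that sways its candy canes and changes its look when the wearer scores a kill.
# Two chained AnimDynamics nodes give the candy cane joints world-space, sprung, angle-limited motion.
# A cosmetic stat watcher on AthenaKills sets Active Kill when the count exceeds the last one.
# Each update swaps "Hide Element" 01/02 material parameters from that flag, which clears after a four-second delay.
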AnimBlueprint /BRCosmetics/Accessories/FORT_Backpacks/Backpack_F_MED_Renegade_Raider_Holiday/Meshes/F_MED_Renegade_Raider_Holiday_Pack_AnimBP
target_skeleton=&/F_MED_Renegade_Raider_Holiday_Pack_Skeleton
graphs [
 AnimGraph {nodes [
   Local Space Ref Pose()
   n1 = Anim Dynamics - Bone: dyn_candyCanes_jnt_1(ComponentPose=Local To Component(LocalPose=Input Pose()), Alpha=1, ExternalForce="0,0,0") {
    angularDampingOverride=0.985
    angularSpringConstant=275
    bAngularSpring=true
    bGravityOverrideInSimSpace=true
    bOverrideAngularDamping=true
    simulationSpace=World
    alphaBoolBlend {blendOption=Linear}
    boundBone {boneName=dyn_candyCanes_jnt_1}
    physicsBodyDefinitions [{
      sphereCollisionRadius=0
      boundBone {boneName=dyn_candyCanes_jnt_1}
      boxExtents=<4 4.5 8>
      constraintSetup {angularTargetAxis=AxisZ, angularLimitsMax=<0 5 1.5>, angularLimitsMin=<0 -5 -1.5>, angularTarget=<0 0 1>}
      localJointOffset=<0 0 0>
     }]
    relativeSpaceBone {boneName=root}
   }
   n2 = Anim Dynamics - Bone: dyn_candyCanes_jnt_2(ComponentPose=@n1, Alpha=1, ExternalForce="0,0,0") {
    angularDampingOverride=0.945
    angularSpringConstant=430
    bAngularSpring=true
    bGravityOverrideInSimSpace=true
    bOverrideAngularDamping=true
    simulationSpace=World
    alphaBoolBlend {blendOption=Linear}
    boundBone {boneName=dyn_candyCanes_jnt_2}
    componentLinearAccScale=<0.05 0.05 0.05>
    componentLinearVelScale=<0.05 0.05 0.05>
    physicsBodyDefinitions [{
      sphereCollisionRadius=0
      boundBone {boneName=dyn_candyCanes_jnt_2}
      boxExtents=<4 4.5 8>
      constraintSetup {angularTargetAxis=AxisZ, angularLimitsMax=<0 5 1.5>, angularLimitsMin=<0 -5 -1.5>, angularTarget=<0 0 1>}
      localJointOffset=<0 0 0>
     }]
    relativeSpaceBone {boneName=root}
   }
   Output Pose(Result=Component To Local(ComponentPose=@n2)) {layerGroup=None}
  ]}
 EventGraph {nodes [
   Event BlueprintUpdateAnimation()
   > Get!(execute:Exec) -> (then:Exec, else:Exec, Skel Mesh:Skeletal Mesh Component Object Reference)
   > Get!(execute) -> (then, else, Pawn:Fort Player Pawn Object Reference)
   > Sequence()
   > n1 = Get Dynamic Material Instance(SkeletalMeshComponent=GetPartSkeletalMeshComponent())
   > n2 = Branch(Condition=Get Active Kill() -> (Active Kill:bool))
   > SetScalarParameterValue(self=@n1, ParameterName=Hide Element 01, Value=1)
   > SetScalarParameterValue(self=@n1, ParameterName=Hide Element 02, Value=0)
   > Delay(Duration=4)
   > Set Active Kill(Active Kill:bool=false) -> (Output_Get:bool)
   SetScalarParameterValue(execute=@n2.else, self=@n1, ParameterName=Hide Element 02, Value=1)
   > SetScalarParameterValue(self=@n1, ParameterName=Hide Element 01, Value=0)
   Event BlueprintInitializeAnimation()
   > Set Skel Mesh(Skel Mesh=GetOwnerSkeletalMeshComponent())
   > n3 = Cast To FortPlayerPawn(Object=TryGetFortPawn())
   > Set Pawn(Pawn:Fort Player Pawn Object Reference=@n3) -> (Output_Get:Fort Player Pawn Object Reference)
   Event OnCharacterCustomizationLoaded()
   > n4 = WatchCosmeticStat!(
    execute:Exec
    Pawn:Fort Player Pawn Object Reference=Get Pawn() -> (Pawn:Fort Player Pawn Object Reference)
    StatTag:Fort Stat Manager Tag Structure=(TagName="StatManager.AthenaKills")
    DefaultValue:int=0
    bBroadcastInitialValue:bool=true
    InitialValueBroadcastDelay:float=-1
    ChangeValueBroadcastDelay:float=-1
    bConsumeWaitingNotifications:bool=true
   ) -> (then:Exec, OnStatChanged:Exec, StatName:Name, StatValue:int) {is=title K2Node_AsyncAction}
   n5 = Set Kill Count(execute=@n4.OnStatChanged, Kill Count:double="To Float (Integer)"(InInt=@n4.StatValue)) -> (Output_Get:double)
   > Branch(Condition="float > float"(A=@n5, B=Get Last Kill Count() -> (Last Kill Count:double)) {is=title K2Node_CallFunction})
   > Set Active Kill(Active Kill=true)
   > Set Last Kill Count(Last Kill Count:double=@n5) -> (Output_Get:double)
   TryGetPawnOwner()
   SetScalarParameterValue(self=@n1, ParameterName=Hide Element 03, Value=1)
  ]}
]
variables [
 Active Kill {type=boolean}
 Kill Count {type=number}
 Last Kill Count {type=number}
 Pawn {type=FortPlayerPawn}
 Skel Mesh {type=SkeletalMeshComponent}
]
